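# Idle sparkle effect for the FNCS Champion pickaxe, built from one Niagara sprite emitter.
# Short-lived, hue-varied particles spawn continuously and take their positions from a baked array of local points instead of sampling a mesh.
# Curves make alpha flicker and ramp in with emitter age, then a shared time-of-day system module scales color and alpha.
# Per-particle center, corona and flicker flare attributes are derived alongside three sprite renderers.
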NiagaraSystem /BRCosmetics/Weapons/FORT_Melee/Pickaxe_FNCSChampion/FX/NS_Pickaxe_FNCSChampion_Idle
dependencies {
 usedDataInterfaces [&NiagaraDataInterfaceCurve, &NiagaraDataInterfaceArrayFloat3, &NiagaraDataInterfaceColorCurve]
 usedDynamicInputs []
 usedModules [&SystemState, /BRCosmetics/Characters/Player/Female/Medium/Bodies/F_MED_PiperShelf_Pearl/FX/TimeOfDayGlobal]
}
emitters {Sparkles {
  emitterUpdateScript [
   &EmitterState()
   &SpawnRate(SpawnRate=30, Use Spawn Probability=false)
  ]
  particleSpawnScript [
   &InitializeParticle(
    AdjustAlpha=false
    AdjustHue=true
    AdjustSaturation=false
    AdjustValue=false
    Color=<rgba 0.843 0.8 1 1>
    Color Mode=::NewEnumerator3
    Lifetime Max=0.7
    Lifetime Min=0.4
    Lifetime Mode=::NewEnumerator1
    Sprite Size Max=<15 35>
    Sprite Size Min=<10 20>
    Sprite Size Mode=::NewEnumerator2
    UsePositionOffset=true
    Write Material Random=true
   ) {module=InitializeParticle001}
   &CameraOffset(Camera Offset Amount=15)
  ]
  particleUpdateScript [
   &ParticleState()
   &ScaleSpriteSize() {module=ScaleSpriteSize001}
   /SetVariables(Particles.Position=&SelectVectorFromArray(
     Array Sampling Mode=::NewEnumerator0
     Direct Array Index=$Particles.SpawnGroup
     Vector Selection Array=DataInterface() {propertyValues {FloatData [
        <4.1968 -12.2851 55.2611>
        <4.9279 -18.3605 62.5095>
        <3.4682 25.6589 75.2552>
        <3.593 18.0839 66.1207>
        <3.0431 20.6243 69.8765>
        <5.2268 -4.9911 76.5552>
        <4.1198 -6.4273 75.9153>
        <3.2548 -3.7581 69.4206>
        <-2.2689 -3.9288 68.5019>
        <-1.032 -4.1998 69.7262>
        <-2.0832 20.2996 67.5504>
        <-3.5937 24.621 77.0654>
        <0.7356 19.9916 63.3745>
        <0.536 21.0897 59.3947>
        <0.7307 16.9272 57.225>
        <0.1351 31.0062 75.637>
        <-3.4092 5.745 35.7089>
        <-3.3165 7.1937 29.3151>
        <-2.0609 9.8973 32.9513>
        <6.8569 6.3043 36.619>
        <-1.1502 12.7577 33.7755>
        <-0.6776 12.9592 28.6403>
        <2.9213 11.4728 23.0205>
        <3.4462 13.1836 33.844>
        <4.0419 11.2912 30.7468>
        <4.597 10.6212 25.3243>
        <0.2423 3.0404 29.0091>
        <-1.8188 3.0533 33.8025>
        <-3.158 2.6951 35.6891>
        <0.4574 7.8207 39.5368>
        <-1.5955 8.9597 43.4567>
        <0.4159 24.6501 51.7334>
        <1.2489 -3.8653 47.7444>
        <-1.1015 32.0411 59.733>
        <-4.9507 4.555 -71.9116>
        <-3.0335 5.7342 -56.24>
        <1.7749 6.0849 -17.6782>
        <2.8017 8.5204 -18.735>
        <4.0213 2.9118 47.4705>
        <2.0852 7.3684 38.9572>
        <3.6462 1.265 55.5661>
        <1.9175 6.7404 42.9747>
        <1.4369 6.9749 42.3692>
        <1.6633 8.7434 50.9326>
        <2.9582 9.1512 51.9738>
        <2.0533 6.5227 49.7868>
        <1.548 8.9429 49.3526>
        <2.6459 8.0432 35.1806>
        <1.1671 7.4986 44.1596>
        <-1.447 8.253 48.783>
        <0.8099 3.5209 48.8686>
        <1.2379 6.6823 38.4201>
        <-0.2652 4.5201 44.7664>
        <0.0443 7.7687 41.0731>
        <-0.3125 8.4519 35.304>
        <-0.0008 14.7478 57.0692>
        <0.3449 8.3388 44.7293>
        <1.2526 5.7464 44.7776>
        <1.1332 6.3669 38.1595>
        <-1.8221 2.549 55.3514>
        <0.9021 7.569 36.3619>
        <1.3897 27.9739 53.1466>
        <2.0992 29.4692 51.2975>
        <-0.822 20.1237 49.5001>
        <1.9593 -2.1504 48.0967>
        <0.4622 -2.9367 46.8764>
        <-0.4411 -9.5356 54.2461>
        <1.312 -10.8966 52.3358>
        <1.9556 -10.2074 52.1412>
        <3.721 -9.8554 52.6712>
        <0.5102 29.5682 71.3092>
        <0.4218 -24.0241 90.057>
        <2.1293 36.008 83.0503>
        <5.3714 5.6719 -63.3911>
        <4.0547 5.9394 -67.5461>
        <1.4396 3.0418 -55.6304>
        <-1.1923 5.9929 -62.8468>
        <0.7963 12.4446 -71.188>
        <3.1528 11.3898 53.0408>
        <-0.3344 19.5781 48.6602>
        <2.1027 6.4224 90.0257>
        <3.5715 8.4642 89.5851>
        <1.4261 -10.4827 71.2385>
        <4.0784 11.9525 51.8511>
        <-0.1128 3.5395 46.0625>
        <-0.4068 8.5313 -63.8777>
        <-1.2388 6.8106 -57.0408>
        <2.3759 6.1788 -61.2897>
        <-0.103 9.3283 24.7156>
        <-0.4096 9.1758 26.8333>
        <1.9326 5.8727 38.9928>
        <-0.4835 8.4751 12.1215>
        <2.731 7.8853 -15.1821>
        <2.7926 7.4283 7.4916>
        <2.1167 9.0671 -5.9265>
        <2.3772 5.9695 82.512>
        <2.0148 -1.5035 69.2649>
        <1.9341 14.465 66.5972>
        <2.1048 15.0438 72.3402>
        <1.8997 3.8473 66.1235>
        <2.0467 8.1023 71.3829>
        <1.7827 14.2594 61.9913>
        <-0.132 10.6494 82.0617>
        <-0.1155 1.0744 81.1657>
        <0.0068 19.8351 76.4602>
        <0.1381 17.9829 72.0691>
        <-0.0831 -5.6703 79.4641>
        <-0.0395 -3.2099 78.044>
        <0.2818 8.6272 68.4751>
        <0.2527 4.1767 68.7517>
        <0.4815 1.8915 61.2964>
        <2.5902 17.2521 64.9544>
        <-0.349 13.0899 59.0398>
        <-0.0616 -0.8505 71.4794>
        <-0.8954 -0.6864 74.0683>
        <-1.0413 -1.8619 77.8856>
        <-1.071 0.954 78.3746>
        <2.7564 11.2315 84.0531>
        <-0.4309 15.1362 82.2237>
        <2.125 9.6722 64.8597>
        <2.5975 9.5346 63.936>
        <-0.8157 15.4162 72.1176>
        <2.8211 1.1271 69.2689>
       ]}}
     Random Seed=0
    )) {isSetParameters=true}
   &ScaleColor(
    Scale Alpha=&Multiply_Float(
     A=&FloatFromCurve(
      FloatCurve=DataInterface() {propertyValues {Curve {keys [
          {interpMode=RCIM_Cubic, tangentMode=RCTM_User, time=0, value=0}
          {arriveTangent=-2.4058, interpMode=RCIM_Cubic, leaveTangent=-2.4058, tangentMode=RCTM_User, time=0.26, value=0.8421}
          {arriveTangent=-1.1775, interpMode=RCIM_Cubic, leaveTangent=-1.1775, time=0.4613, value=0.0645}
          {arriveTangent=-0.1197, interpMode=RCIM_Cubic, leaveTangent=-0.1197, time=0.5898, value=0.4537}
          <at 1 0 RCIM_Cubic>
         ]}}}
      Scale Curve=&RandomRangeFloat(Minimum=0.22, Maximum=0.31)
     )
     B=&FloatFromCurve(
      FloatCurve=DataInterface() {propertyValues {Curve {keys [{tangentMode=RCTM_User, time=0, value=0}, {tangentMode=RCTM_User, time=1, value=1}]}}}
      Scale Curve=&FloatFromCurve(
       FloatCurve=DataInterface() {propertyValues {Curve {keys [
           {arriveTangentWeight=0.1667, interpMode=RCIM_Cubic, leaveTangentWeight=0.1667, tangentMode=RCTM_User, time=1.4, value=0}
           {tangentMode=RCTM_User, time=2, value=1}
          ]}}}
       CurveIndex=$Emitter.Age
      )
     )
    )
    Scale RGB=&VectorFromFloat(Value=&RandomRangeFloat(Minimum=0.6, Maximum=1, Evaluation Type=::NewEnumerator1))
    ScaleA=true
    ScaleRGB=true
   ) {module=ScaleColor001}
   &ScaleColor(
    Scale Alpha=$System.TODAlpha
    Scale RGB=&MakeVectorFromLinearColorRGB(Color=$System.TODColor)
    ScaleA=true
    ScaleRGB=true
   )
   /SetVariables(
    Particles.CenterFlareColor:LinearColor=&MultiplyLinearColor(
     Linear Color=$Particles.Color
     Scale Factor=&MakeVector4_FromLinearColor(LinearColor=<rgba 8.0099 9.8007 10 1>)
    )
    Particles.CenterFlareSize:Vector2f=&Multiply_Vector2(A=$Particles.SpriteSize, B=&Vector2DFromFloat(Value=0.5))
    Particles.CoronaFlareColor:LinearColor=&MultiplyLinearColor(
     Linear Color=$Particles.Color
     Scale Factor=&MakeVector4_FromLinearColor(LinearColor=<rgba 0.15 0.15 0.15 1>)
    )
    Particles.CoronaFlareSize:Vector2f=&Multiply_Vector2(A=$Particles.Initial.SpriteSize, B=&Vector2DFromFloat(Value=1))
    Particles.FlickerFlareColor:LinearColor=&MultiplyLinearColor(
     Linear Color=$Particles.Color
     Scale Factor=&MakeVector4_FromLinearColor(LinearColor=<rgba 0.9441 0.8079 1 1>)
    )
   ) {isSetParameters=true}
  ]
  rendererClasses [&NiagaraSpriteRendererProperties]
  renderers [
   {rendererClass=&NiagaraSpriteRendererProperties}
   {rendererClass=&NiagaraSpriteRendererProperties, rendererIndex=1}
   {rendererClass=&NiagaraSpriteRendererProperties, rendererIndex=2}
  ]
 }}
summary {
 systemName=NS_Pickaxe_FNCSChampion_Idle
 emitters [{emitterName=Sparkles, rendererClasses [&NiagaraSpriteRendererProperties]}]
 userVariables [
  User.bIsFrontEnd {type=NiagaraBool}
  User.bIsFrontEndPreview {type=NiagaraBool}
  User.bisSwinging {type=NiagaraBool}
  User.Lifetime_Scale {type=NiagaraFloat, defaultValue=0, description=""}
 ]
}
