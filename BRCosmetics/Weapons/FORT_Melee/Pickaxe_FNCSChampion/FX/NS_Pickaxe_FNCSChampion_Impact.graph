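# Niagara impact effect for the FNCS Champion pickaxe, judging by its name and path.
# It layers nine emitters named for flare and light, smoke puff, dark background, pink and green spikes, star spike, wave, sparks and confetti.
# System scripts use System State and a spawn-time variable assignment; two user bools flag front-end and front-end preview contexts, though their consumers are not shown.
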NiagaraSystem /BRCosmetics/Weapons/FORT_Melee/Pickaxe_FNCSChampion/FX/NS_Pickaxe_FNCSChampion_Impact
dependencies {
 usedDataInterfaces []
 usedDynamicInputs []
 usedModules [
  &self:SystemSpawnScript.SystemScriptSource.SystemScriptGraph.NiagaraNodeAssignment_0.SetVariables_#1
  &SystemState
 ]
}
emitters {
 ConfettiLWE {rendererClasses [], renderers []}
 DarkBG {rendererClasses [], renderers []}
 FlareAndLight {rendererClasses [], renderers []}
 SmokePuff {rendererClasses [], renderers []}
 Sparks {rendererClasses [], renderers []}
 SpikeGreen {rendererClasses [], renderers []}
 SpikePink {rendererClasses [], renderers []}
 StarSpike {rendererClasses [], renderers []}
 Wave {rendererClasses [], renderers []}
}
summary {
 systemName=NS_Pickaxe_FNCSChampion_Impact
 emitters [
  {emitterName=FlareAndLight, rendererClasses []}
  {emitterName=SmokePuff, rendererClasses []}
  {emitterName=SpikePink, rendererClasses []}
  {emitterName=StarSpike, rendererClasses []}
  {emitterName=DarkBG, rendererClasses []}
  {emitterName=SpikeGreen, rendererClasses []}
  {emitterName=Wave, rendererClasses []}
  {emitterName=Sparks, rendererClasses []}
  {emitterName=ConfettiLWE, rendererClasses []}
 ]
 userVariables [
  User.bIsFrontEnd {type=NiagaraBool}
  User.bIsFrontEndPreview {type=NiagaraBool}
 ]
}
