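# Niagara swing trail effect for the FNCS Champion pickaxe, built from two animation-trail ribbon emitters, Trail and Soft.
# Both spawn ribbons via anim-trail modules; Soft sets its own width mode and trail width and starts with a dimmer color.
# Color shifts through blue, purple, magenta and cyan over particle age, tinted by a system-level time-of-day color.
# Alpha fades out over particle life, scaled by an emitter-age curve and a time-of-day scalar.
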NiagaraSystem /BRCosmetics/Weapons/FORT_Melee/Pickaxe_FNCSChampion/FX/NS_Pickaxe_FNCSChampion_Trail
dependencies {
 usedDataInterfaces [&NiagaraDataInterfaceColorCurve, &NiagaraDataInterfaceCurve, &NiagaraDataInterfaceSocketReader]
 usedDynamicInputs [&FNTimeOfDayColor]
 usedModules [
  &self:SystemSpawnScript.SystemScriptSource.SystemScriptGraph.NiagaraNodeAssignment_1.SetVariables_#1
  &self:SystemSpawnScript.SystemScriptSource.SystemScriptGraph.NiagaraNodeAssignment_0.SetVariables_#2
  &SystemState
  &self:SystemSpawnScript.SystemScriptSource.SystemScriptGraph.NiagaraNodeAssignment_2.SetVariables_#3
 ]
}
emitters {
 Soft {
  emitterSpawnScript [&FN_RibbonIDOffsetUtility()]
  emitterUpdateScript [
   &EmitterState()
   &FN_SpawnAnimTrail(AnimTrailWidthMode=::FromSecond)
  ]
  particleSpawnScript [
   &FN_InitializeAnimTrailBoneIndecies()
   &FN_InitializeAnimTrail(AnimTrailWidth=0.14, Color=<rgba 0.5 0.5 0.5 1>, Lifetime=0.3)
  ]
  particleUpdateScript [
   &SolveForcesAndVelocity(Clamp Velocity=false, Limit Acceleration=false)
   &FN_UpdateAnimTrail() {module=FN_UpdateAnimTrail001}
   &ScaleColor(
    Scale Alpha=&FloatFromCurve(
     FloatCurve=DataInterface() {propertyValues {Curve {keys [
         <at 0.0228 0 RCIM_Cubic>
         {interpMode=RCIM_Cubic, leaveTangent=-0.5945, tangentMode=RCTM_Break, time=0.0678, value=1.5}
         <at 0.36 0.0233>
         {arriveTangent=-0.2139, interpMode=RCIM_Cubic, tangentMode=RCTM_Break, time=1, value=0}
        ]}}}
     Scale Curve=&FloatFromCurve(
      FloatCurve=DataInterface() {propertyValues {Curve {keys [<at 0.1 1 RCIM_Cubic>,<at 0.25 1 RCIM_Cubic>]}}}
      CurveIndex=$Emitter.Age
      Scale Curve=&Multiply_Float(
       A=&FNTimeOfDayScalar(Night Scale=0.35, Morning Scale=0.45, Day Scale=0.5, Evening Scale=0.4)
       B=1
      )
     )
    )
    Scale RGB=&MakeVectorFromLinearColorRGB(Color=&MultiplyLinearColor(
      Linear Color=&ColorFromCurve(
       ColorCurve=DataInterface() {propertyValues {
         AlphaCurve {keys [<at 0 1 RCIM_Cubic>,<at 1 1 RCIM_Cubic>]}
         BlueCurve {keys [<at 0 1 RCIM_Cubic>,<at 0.0899 1>,<at 0.195 1>,<at 0.3068 1>,<at 1 1 RCIM_Cubic>]}
         GreenCurve {keys [<at 0 0.3754 RCIM_Cubic>,<at 0.0899 0.3306>,<at 0.195 0.1146>,<at 0.3068 0.0249>,<at 1 0.7501 RCIM_Cubic>]}
         RedCurve {keys [<at 0 0 RCIM_Cubic>,<at 0.0899 0.243>,<at 0.195 0.3376>,<at 0.3068 0.9775>,<at 1 0.0997 RCIM_Cubic>]}
        }}
       CurveIndex=&Divide_Float(A=$Particles.NormalizedAge, B=1)
      )
      Scale Factor=&MakeVector4_FromLinearColor(LinearColor=$System.ToDColor)
     ))
    ScaleA=true
    ScaleRGB=true
   )
  ]
  rendererClasses [&NiagaraRibbonRendererProperties]
  renderers [{rendererClass=&NiagaraRibbonRendererProperties}]
 }
 Trail {
  emitterSpawnScript [&FN_RibbonIDOffsetUtility()]
  emitterUpdateScript [
   &EmitterState()
   &FN_SpawnAnimTrail()
  ]
  particleSpawnScript [
   &FN_InitializeAnimTrailBoneIndecies()
   &FN_InitializeAnimTrail(Color=<rgba 2 2 2 1>, Lifetime=0.3)
  ]
  particleUpdateScript [
   &SolveForcesAndVelocity(Clamp Velocity=false, Limit Acceleration=false)
   &FN_UpdateAnimTrail() {module=FN_UpdateAnimTrail001}
   &ScaleColor(
    Scale Alpha=&FloatFromCurve(
     FloatCurve=DataInterface() {propertyValues {Curve {keys [
         {arriveTangent=-5.4726, interpMode=RCIM_Cubic, leaveTangent=-5.4726, tangentMode=RCTM_User, time=0, value=1.5}
         <at 0.36 0>
         {arriveTangent=-0.2139, interpMode=RCIM_Cubic, tangentMode=RCTM_Break, time=1, value=0}
        ]}}}
     Scale Curve=&FloatFromCurve(
      FloatCurve=DataInterface() {propertyValues {Curve {keys [<at 0.1 1 RCIM_Cubic>,<at 0.25 1 RCIM_Cubic>]}}}
      CurveIndex=$Emitter.Age
      Scale Curve=&Multiply_Float(
       A=&FNTimeOfDayScalar(Night Scale=0.35, Morning Scale=0.45, Day Scale=0.5, Evening Scale=0.4)
       B=1
      )
     )
    )
    Scale RGB=&MakeVectorFromLinearColorRGB(Color=&MultiplyLinearColor(
      Linear Color=&ColorFromCurve(
       ColorCurve=DataInterface() {propertyValues {
         AlphaCurve {keys [<at 0 1 RCIM_Cubic>,<at 1 1 RCIM_Cubic>]}
         BlueCurve {keys [<at 0 1 RCIM_Cubic>,<at 0.0899 1>,<at 0.195 1>,<at 0.3068 1>,<at 1 1 RCIM_Cubic>]}
         GreenCurve {keys [<at 0 0.3754 RCIM_Cubic>,<at 0.0899 0.3306>,<at 0.195 0.1146>,<at 0.3068 0.0249>,<at 1 0.7501 RCIM_Cubic>]}
         RedCurve {keys [<at 0 0 RCIM_Cubic>,<at 0.0899 0.243>,<at 0.195 0.3376>,<at 0.3068 0.9775>,<at 1 0.0997 RCIM_Cubic>]}
        }}
       CurveIndex=&Divide_Float(A=$Particles.NormalizedAge, B=1)
      )
      Scale Factor=&MakeVector4_FromLinearColor(LinearColor=$System.ToDColor)
     ))
    ScaleA=true
    ScaleRGB=true
   )
  ]
  rendererClasses [&NiagaraRibbonRendererProperties]
  renderers [{rendererClass=&NiagaraRibbonRendererProperties}]
 }
}
summary {
 systemName=NS_Pickaxe_FNCSChampion_Trail
 emitters [
  {emitterName=Trail, rendererClasses [&NiagaraRibbonRendererProperties]}
  {emitterName=Soft, rendererClasses [&NiagaraRibbonRendererProperties]}
 ]
 userVariables [
  User.bIsFrontEnd {type=NiagaraBool}
  User.bIsFrontEndPreview {type=NiagaraBool}
  User.PhysicalSurface {type=NiagaraInt32}
  User.RibbonIDOffset {type=NiagaraInt32}
 ]
}
