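# Animation Blueprint for the FNCS Champion pickaxe that makes its seven pieces and six laurels levitate and react to motion.
# Each update sets per-part offsets from phase-shifted sine waves of real time.
# The AnimGraph adds these to base bones from the input pose, moving pieces in X/Z and laurels in Z.
# Cached poses then feed chained AnimDynamics linear springs on laurels, then pieces, simulated root-relative with constrained travel.
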AnimBlueprint /BRCosmetics/Weapons/FORT_Melee/Pickaxe_FNCSChampion/Meshes/FNCSChampion_Axe_AnimBP
target_skeleton=&/FNCSChampion_Axe_Skeleton
graphs [
 AnimGraph {nodes [
   n1 = Anim Dynamics - Bone: dyn_piece_01(ComponentPose=Local To Component(LocalPose=Use cached pose 'luerels Anim dynamics'()), Alpha=1, ExternalForce="0,0,0") {
    angularDampingOverride=0.99
    angularSpringConstant=5000
    bLinearSpring=true
    bOverrideLinearDamping=true
    bUseGravityOverride=true
    bUsePlanarLimit=false
    gravityScale=0.35
    linearDampingOverride=0.9999
    linearSpringConstant=90
    numSolverIterationsPreUpdate=1
    simulationSpace=RootRelative
    alphaBoolBlend {blendOption=Linear}
    alphaScaleBiasClamp {scale=0.2}
    boundBone {boneName=dyn_piece_01}
    componentAppliedLinearAccClamp=<1300 1300 1300>
    componentLinearAccScale=<0.5 0.5 0.5>
    componentLinearVelScale=<0.5 0.5 0.5>
    gravityOverride {z=-980}
    physicsBodyDefinitions [{
      boundBone {boneName=dyn_piece_01}
      boxExtents=<10 10 10>
      constraintSetup {angularLimitsMax=<0 0 0>, angularLimitsMin=<0 0 0>, angularTarget=<0 0 0>, linearAxesMax=<0.5 1.5 0>, linearAxesMin=<-3 -1.5 0>}
      localJointOffset=<0.2 0.2 0>
     }]
    relativeSpaceBone {boneName=dyn_laurel_base_01}
   }
   n2 = Anim Dynamics - Bone: dyn_piece_02(ComponentPose=@n1, Alpha=1, ExternalForce="0,0,0") {
    angularDampingOverride=0.99
    angularSpringConstant=5000
    bLinearSpring=true
    bOverrideLinearDamping=true
    bUseGravityOverride=true
    bUsePlanarLimit=false
    gravityScale=0.35
    linearDampingOverride=0.9999
    linearSpringConstant=90
    numSolverIterationsPreUpdate=1
    simulationSpace=RootRelative
    alphaBoolBlend {blendOption=Linear}
    alphaScaleBiasClamp {scale=0.2}
    boundBone {boneName=dyn_piece_02}
    componentAppliedLinearAccClamp=<1300 1300 1300>
    componentLinearAccScale=<0.5 0.5 0.5>
    componentLinearVelScale=<0.5 0.5 0.5>
    gravityOverride {z=-980}
    physicsBodyDefinitions [{
      boundBone {boneName=dyn_piece_02}
      boxExtents=<10 10 10>
      constraintSetup {angularLimitsMax=<0 0 0>, angularLimitsMin=<0 0 0>, angularTarget=<0 0 0>, linearAxesMax=<0.5 1.5 0>, linearAxesMin=<-3 -1.5 0>}
      localJointOffset=<0.2 0.2 0>
     }]
    relativeSpaceBone {boneName=dyn_laurel_base_01}
   }
   n3 = Anim Dynamics - Bone: dyn_piece_03(ComponentPose=@n2, Alpha=1, ExternalForce="0,0,0") {
    angularDampingOverride=0.99
    angularSpringConstant=5000
    bLinearSpring=true
    bOverrideLinearDamping=true
    bUseGravityOverride=true
    bUsePlanarLimit=false
    gravityScale=0.35
    linearDampingOverride=0.9999
    linearSpringConstant=90
    numSolverIterationsPreUpdate=1
    simulationSpace=RootRelative
    alphaBoolBlend {blendOption=Linear}
    alphaScaleBiasClamp {scale=0.2}
    boundBone {boneName=dyn_piece_03}
    componentAppliedLinearAccClamp=<1300 1300 1300>
    componentLinearAccScale=<0.5 0.5 0.5>
    componentLinearVelScale=<0.5 0.5 0.5>
    gravityOverride {z=-980}
    physicsBodyDefinitions [{
      boundBone {boneName=dyn_piece_03}
      boxExtents=<10 10 10>
      constraintSetup {angularLimitsMax=<0 0 0>, angularLimitsMin=<0 0 0>, angularTarget=<0 0 0>, linearAxesMax=<0.5 1.5 0>, linearAxesMin=<-3 -1.5 0>}
      localJointOffset=<0.2 0.2 0>
     }]
    relativeSpaceBone {boneName=dyn_laurel_base_01}
   }
   n4 = Anim Dynamics - Bone: dyn_piece_04(ComponentPose=@n3, Alpha=1, ExternalForce="0,0,0") {
    angularDampingOverride=0.99
    angularSpringConstant=5000
    bLinearSpring=true
    bOverrideLinearDamping=true
    bUseGravityOverride=true
    bUsePlanarLimit=false
    gravityScale=0.35
    linearDampingOverride=0.9999
    linearSpringConstant=90
    numSolverIterationsPreUpdate=1
    simulationSpace=RootRelative
    alphaBoolBlend {blendOption=Linear}
    alphaScaleBiasClamp {scale=0.2}
    boundBone {boneName=dyn_piece_04}
    componentAppliedLinearAccClamp=<1300 1300 1300>
    componentLinearAccScale=<0.5 0.5 0.5>
    componentLinearVelScale=<0.5 0.5 0.5>
    gravityOverride {z=-980}
    physicsBodyDefinitions [{
      boundBone {boneName=dyn_piece_04}
      boxExtents=<10 10 10>
      constraintSetup {angularLimitsMax=<0 0 0>, angularLimitsMin=<0 0 0>, angularTarget=<0 0 0>, linearAxesMax=<0.5 1.5 0>, linearAxesMin=<-3 -1.5 0>}
      localJointOffset=<0.2 0.2 0>
     }]
    relativeSpaceBone {boneName=dyn_laurel_base_01}
   }
   n5 = Anim Dynamics - Bone: dyn_piece_05(ComponentPose=@n4, Alpha=1, ExternalForce="0,0,0") {
    angularDampingOverride=0.99
    angularSpringConstant=5000
    bLinearSpring=true
    bOverrideLinearDamping=true
    bUseGravityOverride=true
    bUsePlanarLimit=false
    gravityScale=0.35
    linearDampingOverride=0.9999
    linearSpringConstant=90
    numSolverIterationsPreUpdate=1
    simulationSpace=RootRelative
    alphaBoolBlend {blendOption=Linear}
    alphaScaleBiasClamp {scale=0.2}
    boundBone {boneName=dyn_piece_05}
    componentAppliedLinearAccClamp=<1300 1300 1300>
    componentLinearAccScale=<0.5 0.5 0.5>
    componentLinearVelScale=<0.5 0.5 0.5>
    gravityOverride {z=-980}
    physicsBodyDefinitions [{
      boundBone {boneName=dyn_piece_05}
      boxExtents=<10 10 10>
      constraintSetup {angularLimitsMax=<0 0 0>, angularLimitsMin=<0 0 0>, angularTarget=<0 0 0>, linearAxesMax=<0.5 1.5 0>, linearAxesMin=<-3 -1.5 0>}
      localJointOffset=<0.2 0.2 0>
     }]
    relativeSpaceBone {boneName=dyn_laurel_base_01}
   }
   n6 = Anim Dynamics - Bone: dyn_piece_06(ComponentPose=@n5, Alpha=1, ExternalForce="0,0,0") {
    angularDampingOverride=0.99
    angularSpringConstant=5000
    bLinearSpring=true
    bOverrideLinearDamping=true
    bUseGravityOverride=true
    bUsePlanarLimit=false
    gravityScale=0.35
    linearDampingOverride=0.9999
    linearSpringConstant=90
    numSolverIterationsPreUpdate=1
    simulationSpace=RootRelative
    alphaBoolBlend {blendOption=Linear}
    alphaScaleBiasClamp {scale=0.2}
    boundBone {boneName=dyn_piece_06}
    componentAppliedLinearAccClamp=<1300 1300 1300>
    componentLinearAccScale=<0.5 0.5 0.5>
    componentLinearVelScale=<0.5 0.5 0.5>
    gravityOverride {z=-980}
    physicsBodyDefinitions [{
      boundBone {boneName=dyn_piece_06}
      boxExtents=<10 10 10>
      constraintSetup {angularLimitsMax=<0 0 0>, angularLimitsMin=<0 0 0>, angularTarget=<0 0 0>, linearAxesMax=<0.5 1.5 0>, linearAxesMin=<-3 -1.5 0>}
      localJointOffset=<0.2 0.2 0>
     }]
    relativeSpaceBone {boneName=dyn_laurel_base_01}
   }
   n7 = Anim Dynamics - Bone: dyn_piece_07(ComponentPose=@n6, Alpha=1, ExternalForce="0,0,0") {
    angularDampingOverride=0.99
    angularSpringConstant=5000
    bLinearSpring=true
    bOverrideLinearDamping=true
    bUseGravityOverride=true
    bUsePlanarLimit=false
    gravityScale=0.35
    linearDampingOverride=0.9999
    linearSpringConstant=90
    numSolverIterationsPreUpdate=1
    simulationSpace=RootRelative
    alphaBoolBlend {blendOption=Linear}
    alphaScaleBiasClamp {scale=0.2}
    boundBone {boneName=dyn_piece_07}
    componentAppliedLinearAccClamp=<1300 1300 1300>
    componentLinearAccScale=<0.5 0.5 0.5>
    componentLinearVelScale=<0.5 0.5 0.5>
    gravityOverride {z=-980}
    physicsBodyDefinitions [{
      boundBone {boneName=dyn_piece_07}
      boxExtents=<10 10 10>
      constraintSetup {angularLimitsMax=<0 0 0>, angularLimitsMin=<0 0 0>, angularTarget=<0 0 0>, linearAxesMax=<0.5 1.5 0>, linearAxesMin=<-3 -1.5 0>}
      localJointOffset=<0.2 0.2 0>
     }]
    relativeSpaceBone {boneName=dyn_laurel_base_01}
   }
   Output Pose(Result=Component To Local(ComponentPose=@n7)) {layerGroup=None}
   AnimGraphNode_SubInput_0 = Input Pose()
   n8 = Get Piece01() -> (Piece01:double)
   n9 = "Transform (Modify) Bone - Bone: dyn_piece_base_01"!(
    ComponentPose=Local To Component(LocalPose=@AnimGraphNode_SubInput_0)
    Alpha=1
    Translation_X:double=@n8
    Translation_Y:double=0
    Translation_Z:double=@n8
    Rotation="0,0,0"
    Scale="1,1,1"
   ) -> (Pose) {translationMode=BMM_Additive, translationSpace=BCS_BoneSpace, alphaBoolBlend {blendOption=Linear}, boneToModify {boneName=dyn_piece_base_01}}
   n10 = Get Piece02() -> (Piece02:double)
   n11 = "Transform (Modify) Bone - Bone: dyn_piece_base_02"!(ComponentPose=@n9, Alpha=1, Translation_X:double=@n10, Translation_Y:double=0, Translation_Z:double=@n10, Rotation="0,0,0", Scale="1,1,1") -> (Pose) {translationMode=BMM_Additive, translationSpace=BCS_BoneSpace, alphaBoolBlend {blendOption=Linear}, boneToModify {boneName=dyn_piece_base_02}}
   n12 = Get Piece03() -> (Piece03:double)
   n13 = "Transform (Modify) Bone - Bone: dyn_piece_base_03"!(ComponentPose=@n11, Alpha=1, Translation_X:double=@n12, Translation_Y:double=0, Translation_Z:double=@n12, Rotation="0,0,0", Scale="1,1,1") -> (Pose) {translationMode=BMM_Additive, translationSpace=BCS_BoneSpace, alphaBoolBlend {blendOption=Linear}, boneToModify {boneName=dyn_piece_base_03}}
   n14 = Get Piece04() -> (Piece04:double)
   n15 = "Transform (Modify) Bone - Bone: dyn_piece_base_04"!(ComponentPose=@n13, Alpha=1, Translation_X:double=@n14, Translation_Y:double=0, Translation_Z:double=@n14, Rotation="0,0,0", Scale="1,1,1") -> (Pose) {translationMode=BMM_Additive, translationSpace=BCS_BoneSpace, alphaBoolBlend {blendOption=Linear}, boneToModify {boneName=dyn_piece_base_04}}
   n16 = Get Piece05() -> (Piece05:double)
   n17 = "Transform (Modify) Bone - Bone: dyn_piece_base_05"!(ComponentPose=@n15, Alpha=1, Translation_X:double=@n16, Translation_Y:double=0, Translation_Z:double=@n16, Rotation="0,0,0", Scale="1,1,1") -> (Pose) {translationMode=BMM_Additive, translationSpace=BCS_BoneSpace, alphaBoolBlend {blendOption=Linear}, boneToModify {boneName=dyn_piece_base_05}}
   n18 = Get Piece06() -> (Piece06:double)
   n19 = "Transform (Modify) Bone - Bone: dyn_piece_base_06"!(ComponentPose=@n17, Alpha=1, Translation_X:double=@n18, Translation_Y:double=0, Translation_Z:double=@n18, Rotation="0,0,0", Scale="1,1,1") -> (Pose) {translationMode=BMM_Additive, translationSpace=BCS_BoneSpace, alphaBoolBlend {blendOption=Linear}, boneToModify {boneName=dyn_piece_base_06}}
   n20 = Get Piece07() -> (Piece07:double)
   n21 = "Transform (Modify) Bone - Bone: dyn_piece_base_07"!(ComponentPose=@n19, Alpha=1, Translation_X:double=@n20, Translation_Y:double=0, Translation_Z:double=@n20, Rotation="0,0,0", Scale="1,1,1") -> (Pose) {translationMode=BMM_Additive, translationSpace=BCS_BoneSpace, alphaBoolBlend {blendOption=Linear}, boneToModify {boneName=dyn_piece_base_07}}
   Save cached pose 'Levitation pieces cache'(Pose=Component To Local(ComponentPose=@n21)) {cacheName=Levitation pieces cache}
   n22 = "Transform (Modify) Bone - Bone: dyn_laurel_base_01"!(
    ComponentPose=Local To Component(LocalPose=Use cached pose 'Levitation pieces cache'())
    Alpha=1
    Translation_X:double=0
    Translation_Y:double=0
    Translation_Z:double=Get Laurel01() -> (Laurel01:double)
    Rotation="0,0,0"
    Scale="1,1,1"
   ) -> (Pose) {
    translationMode=BMM_Additive
    translationSpace=BCS_BoneSpace
    alphaBoolBlend {blendOption=Linear}
    boneToModify {boneName=dyn_laurel_base_01}
   }
   n23 = "Transform (Modify) Bone - Bone: dyn_laurel_base_02"!(
    ComponentPose=@n22
    Alpha=1
    Translation_X:double=0
    Translation_Y:double=0
    Translation_Z:double=Get Laurel02() -> (Laurel02:double)
    Rotation="0,0,0"
    Scale="1,1,1"
   ) -> (Pose) {
    translationMode=BMM_Additive
    translationSpace=BCS_BoneSpace
    alphaBoolBlend {blendOption=Linear}
    boneToModify {boneName=dyn_laurel_base_02}
   }
   n24 = "Transform (Modify) Bone - Bone: dyn_laurel_base_03"!(
    ComponentPose=@n23
    Alpha=1
    Translation_X:double=0
    Translation_Y:double=0
    Translation_Z:double=Get Laurel03() -> (Laurel03:double)
    Rotation="0,0,0"
    Scale="1,1,1"
   ) -> (Pose) {
    translationMode=BMM_Additive
    translationSpace=BCS_BoneSpace
    alphaBoolBlend {blendOption=Linear}
    boneToModify {boneName=dyn_laurel_base_03}
   }
   n25 = "Transform (Modify) Bone - Bone: dyn_laurel_base_04"!(
    ComponentPose=@n24
    Alpha=1
    Translation_X:double=0
    Translation_Y:double=0
    Translation_Z:double=Get Laurel04() -> (Laurel04:double)
    Rotation="0,0,0"
    Scale="1,1,1"
   ) -> (Pose) {
    translationMode=BMM_Additive
    translationSpace=BCS_BoneSpace
    alphaBoolBlend {blendOption=Linear}
    boneToModify {boneName=dyn_laurel_base_04}
   }
   n26 = "Transform (Modify) Bone - Bone: dyn_laurel_base_05"!(
    ComponentPose=@n25
    Alpha=1
    Translation_X:double=0
    Translation_Y:double=0
    Translation_Z:double=Get Laurel05() -> (Laurel05:double)
    Rotation="0,0,0"
    Scale="1,1,1"
   ) -> (Pose) {
    translationMode=BMM_Additive
    translationSpace=BCS_BoneSpace
    alphaBoolBlend {blendOption=Linear}
    boneToModify {boneName=dyn_laurel_base_05}
   }
   n27 = "Transform (Modify) Bone - Bone: dyn_laurel_base_06"!(
    ComponentPose=@n26
    Alpha=1
    Translation_X:double=0
    Translation_Y:double=0
    Translation_Z:double=Get Laurel06() -> (Laurel06:double)
    Rotation="0,0,0"
    Scale="1,1,1"
   ) -> (Pose) {
    translationMode=BMM_Additive
    translationSpace=BCS_BoneSpace
    alphaBoolBlend {blendOption=Linear}
    boneToModify {boneName=dyn_laurel_base_06}
   }
   Save cached pose 'Levitation laurels cache'(Pose=Component To Local(ComponentPose=@n27)) {cacheName=Levitation laurels cache}
   n28 = Anim Dynamics - Bone: dyn_laurel_01(ComponentPose=Local To Component(LocalPose=Use cached pose 'Levitation laurels cache'()), Alpha=1, ExternalForce="0,0,0") {
    angularDampingOverride=0.99
    angularSpringConstant=5000
    bLinearSpring=true
    bOverrideLinearDamping=true
    bUseGravityOverride=true
    bUsePlanarLimit=false
    gravityScale=0.35
    linearDampingOverride=0.9999
    linearSpringConstant=70
    numSolverIterationsPreUpdate=1
    simulationSpace=RootRelative
    alphaBoolBlend {blendOption=Linear}
    alphaScaleBiasClamp {scale=0.2}
    boundBone {boneName=dyn_laurel_01}
    componentAppliedLinearAccClamp=<1300 1300 1300>
    componentLinearAccScale=<0.5 0.5 0.5>
    componentLinearVelScale=<0.5 0.5 0.5>
    gravityOverride {z=-980}
    physicsBodyDefinitions [{
      boundBone {boneName=dyn_laurel_01}
      boxExtents=<10 10 10>
      constraintSetup {angularLimitsMax=<0 0 0>, angularLimitsMin=<0 0 0>, angularTarget=<0 0 0>, linearAxesMax=<0 4 0>, linearAxesMin=<0 -4 0>}
      localJointOffset=<0 -1 0>
     }]
    relativeSpaceBone {boneName=dyn_laurel_base_01}
   }
   n29 = Anim Dynamics - Bone: dyn_laurel_02(ComponentPose=@n28, Alpha=1, ExternalForce="0,0,0") {
    angularDampingOverride=0.99
    angularSpringConstant=5000
    bLinearSpring=true
    bOverrideLinearDamping=true
    bUseGravityOverride=true
    bUsePlanarLimit=false
    gravityScale=0.35
    linearDampingOverride=0.9999
    linearSpringConstant=70
    numSolverIterationsPreUpdate=1
    simulationSpace=RootRelative
    alphaBoolBlend {blendOption=Linear}
    alphaScaleBiasClamp {scale=0.2}
    boundBone {boneName=dyn_laurel_02}
    componentAppliedLinearAccClamp=<1300 1300 1300>
    componentLinearAccScale=<0.5 0.5 0.5>
    componentLinearVelScale=<0.5 0.5 0.5>
    gravityOverride {z=-980}
    physicsBodyDefinitions [{
      boundBone {boneName=dyn_laurel_02}
      boxExtents=<10 10 10>
      constraintSetup {angularLimitsMax=<0 0 0>, angularLimitsMin=<0 0 0>, angularTarget=<0 0 0>, linearAxesMax=<0 20 0>, linearAxesMin=<0 -20 0>}
      localJointOffset=<0 -1 0>
     }]
    relativeSpaceBone {boneName=dyn_laurel_base_01}
   }
   n30 = Anim Dynamics - Bone: dyn_laurel_03(ComponentPose=@n29, Alpha=1, ExternalForce="0,0,0") {
    angularDampingOverride=0.99
    angularSpringConstant=5000
    bLinearSpring=true
    bOverrideLinearDamping=true
    bUseGravityOverride=true
    bUsePlanarLimit=false
    gravityScale=0.35
    linearDampingOverride=0.9999
    linearSpringConstant=70
    numSolverIterationsPreUpdate=1
    simulationSpace=RootRelative
    alphaBoolBlend {blendOption=Linear}
    alphaScaleBiasClamp {scale=0.2}
    boundBone {boneName=dyn_laurel_03}
    componentAppliedLinearAccClamp=<1300 1300 1300>
    componentLinearAccScale=<0.5 0.5 0.5>
    componentLinearVelScale=<0.5 0.5 0.5>
    gravityOverride {z=-980}
    physicsBodyDefinitions [{
      boundBone {boneName=dyn_laurel_03}
      boxExtents=<10 10 10>
      constraintSetup {angularLimitsMax=<0 0 0>, angularLimitsMin=<0 0 0>, angularTarget=<0 0 0>, linearAxesMax=<0 10 0>, linearAxesMin=<0 -10 0>}
      localJointOffset=<0 -1 0>
     }]
    relativeSpaceBone {boneName=dyn_laurel_base_01}
   }
   n31 = Anim Dynamics - Bone: dyn_laurel_04(ComponentPose=@n30, Alpha=1, ExternalForce="0,0,0") {
    angularDampingOverride=0.99
    angularSpringConstant=5000
    bLinearSpring=true
    bOverrideLinearDamping=true
    bUseGravityOverride=true
    bUsePlanarLimit=false
    gravityScale=0.35
    linearDampingOverride=0.9999
    linearSpringConstant=70
    numSolverIterationsPreUpdate=1
    simulationSpace=RootRelative
    alphaBoolBlend {blendOption=Linear}
    alphaScaleBiasClamp {scale=0.2}
    boundBone {boneName=dyn_laurel_04}
    componentAppliedLinearAccClamp=<1300 1300 1300>
    componentLinearAccScale=<0.5 0.5 0.5>
    componentLinearVelScale=<0.5 0.5 0.5>
    gravityOverride {z=-980}
    physicsBodyDefinitions [{
      boundBone {boneName=dyn_laurel_04}
      boxExtents=<10 10 10>
      constraintSetup {angularLimitsMax=<0 0 0>, angularLimitsMin=<0 0 0>, angularTarget=<0 0 0>, linearAxesMax=<0 3 0>, linearAxesMin=<0 -3 0>}
      localJointOffset=<0 1 0>
     }]
    relativeSpaceBone {boneName=dyn_laurel_base_01}
   }
   n32 = Anim Dynamics - Bone: dyn_laurel_05(ComponentPose=@n31, Alpha=1, ExternalForce="0,0,0") {
    angularDampingOverride=0.99
    angularSpringConstant=5000
    bLinearSpring=true
    bOverrideLinearDamping=true
    bUseGravityOverride=true
    bUsePlanarLimit=false
    gravityScale=0.35
    linearDampingOverride=0.9999
    linearSpringConstant=70
    numSolverIterationsPreUpdate=1
    simulationSpace=RootRelative
    alphaBoolBlend {blendOption=Linear}
    alphaScaleBiasClamp {scale=0.2}
    boundBone {boneName=dyn_laurel_05}
    componentAppliedLinearAccClamp=<1300 1300 1300>
    componentLinearAccScale=<0.5 0.5 0.5>
    componentLinearVelScale=<0.5 0.5 0.5>
    gravityOverride {z=-980}
    physicsBodyDefinitions [{
      boundBone {boneName=dyn_laurel_05}
      boxExtents=<10 10 10>
      constraintSetup {angularLimitsMax=<0 0 0>, angularLimitsMin=<0 0 0>, angularTarget=<0 0 0>, linearAxesMax=<0 5 0>, linearAxesMin=<0 -5 0>}
      localJointOffset=<0 1 0>
     }]
    relativeSpaceBone {boneName=dyn_laurel_base_01}
   }
   n33 = Anim Dynamics - Bone: dyn_laurel_06(ComponentPose=@n32, Alpha=1, ExternalForce="0,0,0") {
    angularDampingOverride=0.99
    angularSpringConstant=5000
    bLinearSpring=true
    bOverrideLinearDamping=true
    bUseGravityOverride=true
    bUsePlanarLimit=false
    gravityScale=0.35
    linearDampingOverride=0.9999
    linearSpringConstant=70
    numSolverIterationsPreUpdate=1
    simulationSpace=RootRelative
    alphaBoolBlend {blendOption=Linear}
    alphaScaleBiasClamp {scale=0.2}
    boundBone {boneName=dyn_laurel_06}
    componentAppliedLinearAccClamp=<1300 1300 1300>
    componentLinearAccScale=<0.5 0.5 0.5>
    componentLinearVelScale=<0.5 0.5 0.5>
    gravityOverride {z=-980}
    physicsBodyDefinitions [{
      boundBone {boneName=dyn_laurel_06}
      boxExtents=<10 10 10>
      constraintSetup {angularLimitsMax=<0 0 0>, angularLimitsMin=<0 0 0>, angularTarget=<0 0 0>, linearAxesMax=<0 15 0>, linearAxesMin=<0 -15 0>}
      localJointOffset=<0 1 0>
     }]
    relativeSpaceBone {boneName=dyn_laurel_base_01}
   }
   Save cached pose 'luerels Anim dynamics'(Pose=Component To Local(ComponentPose=@n33)) {cacheName=luerels Anim dynamics}
  ]}
 EventGraph {nodes [
   Event BlueprintUpdateAnimation()
   n1 = GetRealTimeSeconds()
   n2 = float + float(A=float * float(A=@n1, B=1) {is=title K2Node_PromotableOperator}, B=20)
   > Set Piece01(Piece01:double=float * float(A="Sin (Radians)"(A=@n2), B=1)) -> (Output_Get:double)
   n3 = float + float(A=float * float(A=@n1, B=1), B=40)
   > Set Piece02(Piece02:double=float * float(A="Sin (Radians)"(A=@n3), B=1.2)) -> (Output_Get:double)
   n4 = float + float(A=float * float(A=@n1, B=1), B=50)
   > Set Piece03(Piece03:double=float * float(A="Sin (Radians)"(A=@n4), B=0.6)) -> (Output_Get:double)
   n5 = float + float(A=float * float(A=@n1, B=1), B=60)
   > Set Piece04(Piece04:double=float * float(A="Sin (Radians)"(A=@n5), B=0.5)) -> (Output_Get:double)
   n6 = float + float(A=float * float(A=@n1, B=1), B=70)
   > Set Piece05(Piece05:double=float * float(A="Sin (Radians)"(A=@n6), B=1)) -> (Output_Get:double)
   n7 = float + float(A=float * float(A=@n1, B=1), B=80)
   > Set Piece06(Piece06:double=float * float(A="Sin (Radians)"(A=@n7), B=1)) -> (Output_Get:double)
   n8 = float + float(A=float * float(A=@n1, B=1), B=90)
   > Set Piece07(Piece07:double=float * float(A="Sin (Radians)"(A=@n8), B=0.5)) -> (Output_Get:double)
   n9 = float + float(A=float * float(A=@n1, B=1), B=60)
   > Set Laurel01(Laurel01:double=float * float(A="Sin (Radians)"(A=@n9), B=0.5)) -> (Output_Get:double)
   n10 = float + float(A=float * float(A=@n1, B=1), B=40)
   > Set Laurel02(Laurel02:double=float * float(A="Sin (Radians)"(A=@n10), B=0.6)) -> (Output_Get:double)
   n11 = float + float(A=float * float(A=@n1, B=1), B=20)
   > Set Laurel03(Laurel03:double=float * float(A="Sin (Radians)"(A=@n11), B=0.8)) -> (Output_Get:double)
   n12 = float + float(A=float * float(A=@n1, B=1), B=60)
   > Set Laurel04(Laurel04:double=float * float(A="Sin (Radians)"(A=@n12), B=0.5)) -> (Output_Get:double)
   n13 = float + float(A=float * float(A=@n1, B=1), B=40)
   > Set Laurel05(Laurel05:double=float * float(A="Sin (Radians)"(A=@n13), B=0.6)) -> (Output_Get:double)
   n14 = float + float(A=float * float(A=@n1, B=1), B=20)
   > Set Laurel06(Laurel06:double=float * float(A="Sin (Radians)"(A=@n14), B=0.8)) -> (Output_Get:double)
   TryGetPawnOwner()
  ]}
]
variables [
 Laurel01 {type=number}
 Laurel02 {type=number}
 Laurel03 {type=number}
 Laurel04 {type=number}
 Laurel05 {type=number}
 Laurel06 {type=number}
 Piece01 {type=number}
 Piece02 {type=number}
 Piece03 {type=number}
 Piece04 {type=number}
 Piece05 {type=number}
 Piece06 {type=number}
 Piece07 {type=number}
]
